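# Physics asset for the Masked Warrior medium male body cosmetic, covering the dynamic front collar flaps and the high and low necklaces.
# Each flap and two-segment necklace chain hangs from a small base body through tight, asymmetric swing and twist limits.
# The base bodies link to a spine_04 box with free angular motion.
# A right upper-arm capsule, also linked freely to spine_04, is the only limb body, with no left-side counterpart.
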PhysicsAsset /BRCosmetics/Characters/Player/Male/Medium/Bodies/M_MED_Masked_Warrior/Meshes/M_MED_MaskedWarrior_CPhysics
bodies {
 dyn_collarFront_Base_l [{shapeType=Box, extentX=1.01, extentY=1.01, extentZ=1.01, rotation=<rot 0 -90 0>}]
 dyn_collarFront_Base_r [{shapeType=Box, extentX=1.01, extentY=1.01, extentZ=1.01, rotation=<rot 0 -90 0>}]
 dyn_collarFront_l [{shapeType=Box, extentX=7.5464, extentY=1.7759, extentZ=10.2637, center=<-3.5831 0.1018 -0.2417>, rotation=<rot 0.6056 -177.1173 100.8896>}]
 dyn_collarFront_r [{shapeType=Box, extentX=7.5464, extentY=1.7759, extentZ=10.2637, center=<3.5831 -0.1018 0.2417>, rotation=<rot 0.6056 -177.1173 -79.1104>}]
 dyn_highNecklace_01 [{shapeType=Capsule, length=6.2248, radius=2.2625, center=<2.049 -0.0001 0>, rotation=<rot 90 0 0>}]
 dyn_highNecklace_01_Base [{shapeType=Capsule, length=0.505, radius=0.505, rotation=<rot 0 -90 0>}]
 dyn_highNecklace_02 [{shapeType=Capsule, length=6.7758, radius=4.7657, center=<5.5781 0.0001 0>, rotation=<rot 90 0 0>}]
 dyn_lowNecklace_01 [{shapeType=Capsule, length=6.7427, radius=3.329, center=<3.2743 0.3685 0>, rotation=<rot 90 0 0>}]
 dyn_lowNecklace_01_Base [{shapeType=Capsule, length=0.505, radius=0.505, rotation=<rot 0 -90 0>}]
 dyn_lowNecklace_02 [{shapeType=Capsule, length=5.6965, radius=3.8601, center=<2.9199 0.2651 0>, rotation=<rot 90 0 0>}]
 spine_04 [{shapeType=Box, extentX=3.0698, extentY=13.1638, extentZ=23.3348, center=<-2.6888 8.101 0.6414>, rotation=<rot 2.1676 -91.0152 2.221>}]
 upperarm_r [{shapeType=Capsule, length=20.0748, radius=5.9041, center=<-14.584 0.2007 0.0001>, rotation=<rot 90 0 0>}]
}
constraints [
 {bone1Name=dyn_lowNecklace_02, bone2Name=dyn_lowNecklace_01, swing1LimitDegrees=2, swing2LimitDegrees=20, twistLimitDegrees=1}
 {bone1Name=dyn_highNecklace_02, bone2Name=dyn_highNecklace_01, swing1LimitDegrees=10, swing2LimitDegrees=15, twistLimitDegrees=2}
 {bone1Name=dyn_highNecklace_01, bone2Name=dyn_highNecklace_01_Base, swing1LimitDegrees=2, swing2LimitDegrees=8, twistLimitDegrees=2}
 {bone1Name=dyn_lowNecklace_01, bone2Name=dyn_lowNecklace_01_Base, swing1LimitDegrees=2, swing2LimitDegrees=15, twistLimitDegrees=1}
 {
  bone1Name=dyn_lowNecklace_01_Base
  bone2Name=spine_04
  swing1LimitDegrees=45
  swing1Motion=Free
  swing2LimitDegrees=45
  swing2Motion=Free
  twistLimitDegrees=45
  twistMotion=Free
 }
 {
  bone1Name=dyn_highNecklace_01_Base
  bone2Name=spine_04
  swing1LimitDegrees=45
  swing1Motion=Free
  swing2LimitDegrees=45
  swing2Motion=Free
  twistLimitDegrees=45
  twistMotion=Free
 }
 {
  bone1Name=dyn_collarFront_Base_r
  bone2Name=spine_04
  swing1LimitDegrees=45
  swing1Motion=Free
  swing2LimitDegrees=45
  swing2Motion=Free
  twistLimitDegrees=45
  twistMotion=Free
 }
 {
  bone1Name=dyn_collarFront_Base_l
  bone2Name=spine_04
  swing1LimitDegrees=45
  swing1Motion=Free
  swing2LimitDegrees=45
  swing2Motion=Free
  twistLimitDegrees=45
  twistMotion=Free
 }
 {bone1Name=dyn_collarFront_l, bone2Name=dyn_collarFront_Base_l, swing1LimitDegrees=8, swing2LimitDegrees=30, twistLimitDegrees=10}
 {bone1Name=dyn_collarFront_r, bone2Name=dyn_collarFront_Base_r, swing1LimitDegrees=8, swing2LimitDegrees=30, twistLimitDegrees=10}
 {
  bone1Name=upperarm_r
  bone2Name=spine_04
  swing1LimitDegrees=45
  swing1Motion=Free
  swing2LimitDegrees=45
  swing2Motion=Free
  twistLimitDegrees=45
  twistMotion=Free
 }
]
