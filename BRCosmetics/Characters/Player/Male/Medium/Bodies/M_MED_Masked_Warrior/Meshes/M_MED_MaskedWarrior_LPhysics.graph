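# Physics asset for the male medium Masked Warrior body, covering the left coattail bone chain and the legs it can collide with.
# Six box bodies, generally shrinking toward the tip, form the coattail, linked in sequence by swing-limited constraints whose twist tightens along the lower links.
# Spheres and capsules approximate both thighs and calves plus the left thigh twist, joined by fully free angular constraints.
# No constraint links the coattail chain to the leg bodies.
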PhysicsAsset /BRCosmetics/Characters/Player/Male/Medium/Bodies/M_MED_Masked_Warrior/Meshes/M_MED_MaskedWarrior_LPhysics
bodies {
 calf_l [
  {shapeType=Sphere, radius=11.8083, center=<-1.0507 -1.7452 0.0002>}
  {shapeType=Sphere, radius=11.6777, center=<-7.5438 -2.9458 -0.0001>}
  {shapeType=Capsule, length=25.8953, radius=7.44, center=<-18.7397 0 0.0001>, rotation=<rot 90 0 180>}
 ]
 calf_r [
  {shapeType=Sphere, radius=8.375, center=<1.0011 0 -1.4352>}
  {shapeType=Capsule, length=23.9547, radius=9.3806, center=<21.5296 2.5447 -1.0583>, rotation=<rot 90 0 0>}
 ]
 dyn_coattail_01_l [{shapeType=Box, extentX=1.9859, extentY=9.4464, extentZ=12.7979, center=<-5.1374 0.3772 0.6868>, rotation=<rot 2.9315 -90 20.8469>}]
 dyn_coattail_02_l [{shapeType=Box, extentX=2.4781, extentY=7.8762, extentZ=12.4437, center=<-5.1659 0.3603 0>, rotation=<rot 0 90 90>}]
 dyn_coattail_03_l [{shapeType=Box, extentX=2.7977, extentY=7.4422, extentZ=13.2077, center=<-6.0654 0.0002 0>, rotation=<rot 0 90 -90>}]
 dyn_coattail_04_l [{shapeType=Box, extentX=1.6374, extentY=4.2386, extentZ=11.3045, center=<-5.0404 0.0467 0.2032>, rotation=<rot 0 -89.7926 90.7571>}]
 dyn_coattail_05_l [{shapeType=Box, extentX=1.5527, extentY=2.8136, extentZ=6.7253, center=<-2.8513 -0.0555 0.1145>, rotation=<rot 0 -89.5344 91.0302>}]
 dyn_coattail_06_l [{shapeType=Box, extentX=1.2019, extentY=1.9667, extentZ=5.7432, center=<-2.613 -0.1125 0.1388>, rotation=<rot 0 -90.1859 91.1411>}]
 thigh_l [
  {shapeType=Sphere, radius=9.2179, center=<-32.7777 0.448 0>}
  {shapeType=Capsule, length=25.7464, radius=8.7793, center=<-19.3555 0 -0.0002>, rotation=<rot 90 0 180>}
 ]
 thigh_r [{shapeType=Capsule, length=25.6335, radius=8.09, center=<20.6151 0.0001 -0.2031>, rotation=<rot 88.7617 0 0>}]
 thigh_twist_01_l [{shapeType=Capsule, length=18.0637, radius=8.9171, center=<-1.6073 0 0.0001>, rotation=<rot 90 0 0>}]
}
constraints [
 {
  bone1Name=calf_l
  bone2Name=thigh_l
  swing1LimitDegrees=45
  swing1Motion=Free
  swing2LimitDegrees=45
  swing2Motion=Free
  twistLimitDegrees=45
  twistMotion=Free
 }
 {bone1Name=dyn_coattail_04_l, bone2Name=dyn_coattail_03_l, swing1LimitDegrees=30, swing2LimitDegrees=25, twistLimitDegrees=5}
 {bone1Name=dyn_coattail_05_l, bone2Name=dyn_coattail_04_l, swing1LimitDegrees=30, swing2LimitDegrees=25, twistLimitDegrees=5}
 {bone1Name=dyn_coattail_06_l, bone2Name=dyn_coattail_05_l, swing1LimitDegrees=30, swing2LimitDegrees=25, twistLimitDegrees=5}
 {bone1Name=dyn_coattail_03_l, bone2Name=dyn_coattail_02_l, swing1LimitDegrees=30, swing2LimitDegrees=25, twistLimitDegrees=10}
 {bone1Name=dyn_coattail_02_l, bone2Name=dyn_coattail_01_l, swing1LimitDegrees=30, swing2LimitDegrees=25, twistLimitDegrees=10}
 {
  bone1Name=thigh_twist_01_l
  bone2Name=thigh_l
  swing1LimitDegrees=45
  swing1Motion=Free
  swing2LimitDegrees=45
  swing2Motion=Free
  twistLimitDegrees=45
  twistMotion=Free
 }
 {
  bone1Name=calf_r
  bone2Name=thigh_r
  swing1LimitDegrees=45
  swing1Motion=Free
  swing2LimitDegrees=45
  swing2Motion=Free
  twistLimitDegrees=45
  twistMotion=Free
 }
]
