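# Physics asset for the Masked Warrior male medium player body, covering the right coattail and right leg.
# Six box bodies form the coattail bone chain, linked neighbor to neighbor by constraints with swing and twist limits, twist slightly tighter near the root.
# Spheres and capsules shape the right thigh and calf, apparently as colliders for the coattail, with a free-rotating constraint between the two.
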PhysicsAsset /BRCosmetics/Characters/Player/Male/Medium/Bodies/M_MED_Masked_Warrior/Meshes/M_MED_MaskedWarrior_RPhysics
bodies {
 calf_r [
  {shapeType=Sphere, radius=11.8083, center=<1.0507 1.7452 -0.0002>}
  {shapeType=Sphere, radius=11.6777, center=<7.5438 2.9458 0.0001>}
  {shapeType=Capsule, length=25.8953, radius=7.44, center=<18.7397 0 -0.0001>, rotation=<rot 90 0 0>}
 ]
 dyn_coattail_01_r [{shapeType=Box, extentX=1.9857, extentY=9.1842, extentZ=4.814, rotation=<rot 0 90 90>}]
 dyn_coattail_02_r [{shapeType=Box, extentX=2.2508, extentY=10.1575, extentZ=12.2128, center=<5.1167 -0.1591 0>, rotation=<rot 0 90 90>}]
 dyn_coattail_03_r [{shapeType=Box, extentX=3.0588, extentY=9.6922, extentZ=13.2077, center=<6.0654 -0.0002 0>, rotation=<rot 0 90 90>}]
 dyn_coattail_04_r [{shapeType=Box, extentX=3.3258, extentY=6.4886, extentZ=11.3045, center=<5.0404 -0.0467 -0.2032>, rotation=<rot 0 -89.7926 -89.2429>}]
 dyn_coattail_05_r [{shapeType=Box, extentX=2.3438, extentY=2.8136, extentZ=6.7253, center=<2.8513 0.0555 -0.1145>, rotation=<rot 0 -89.5344 -88.9698>}]
 dyn_coattail_06_r [{shapeType=Box, extentX=1.3766, extentY=1.9667, extentZ=5.7432, center=<2.613 0.1125 -0.1388>, rotation=<rot 0 -90.1859 -88.8588>}]
 thigh_r [
  {shapeType=Sphere, radius=9.2179, center=<32.7777 -0.448 0>}
  {shapeType=Capsule, length=17.4964, radius=8.9043, center=<21.0969 0 0.0003>, rotation=<rot 90 0 0>}
 ]
}
constraints [
 {bone1Name=dyn_coattail_04_r, bone2Name=dyn_coattail_03_r, swing1LimitDegrees=30, swing2LimitDegrees=25, twistLimitDegrees=10}
 {bone1Name=dyn_coattail_05_r, bone2Name=dyn_coattail_04_r, swing1LimitDegrees=30, swing2LimitDegrees=25, twistLimitDegrees=10}
 {bone1Name=dyn_coattail_06_r, bone2Name=dyn_coattail_05_r, swing1LimitDegrees=30, swing2LimitDegrees=25, twistLimitDegrees=10}
 {
  bone1Name=calf_r
  bone2Name=thigh_r
  swing1LimitDegrees=45
  swing1Motion=Free
  swing2LimitDegrees=45
  swing2Motion=Free
  twistLimitDegrees=45
  twistMotion=Free
 }
 {bone1Name=dyn_coattail_03_r, bone2Name=dyn_coattail_02_r, swing1LimitDegrees=30, swing2LimitDegrees=25, twistLimitDegrees=6}
 {bone1Name=dyn_coattail_02_r, bone2Name=dyn_coattail_01_r, swing1LimitDegrees=30, swing2LimitDegrees=25, twistLimitDegrees=6}
]
