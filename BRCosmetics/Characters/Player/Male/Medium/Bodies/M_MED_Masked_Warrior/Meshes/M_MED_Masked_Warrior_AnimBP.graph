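# Post-process style secondary-motion AnimBP for the Masked Warrior body, copying the parent mesh's pose and animating coattails, bag, collar and shoe tongues.
# Event graph maps thigh, head and arm angles to corrective alphas, and pulses bag force on jumps.
# Cached constraint and AnimDynamics pose feeds grounded/airborne coattail blend, then RigidBody sims, collar correctives and shoe-tongue Look Ats.
# Several older node chains remain disconnected from the output.
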AnimBlueprint /BRCosmetics/Characters/Player/Male/Medium/Bodies/M_MED_Masked_Warrior/Meshes/M_MED_Masked_Warrior_AnimBP
target_skeleton=&/M_MED_Masked_Warrior_Skeleton
graphs [
 AnimGraph {nodes [
   Anim Dynamics - Bone: dyn_coattail_02_r(Alpha=1, ExternalForce="0,0,0") {
    angularDampingOverride=1
    angularSpringConstant=800
    bAngularSpring=true
    bGravityOverrideInSimSpace=true
    bOverrideAngularDamping=true
    bUseSphericalLimits=true
    simulationSpace=RootRelative
    alphaBoolBlend {blendOption=Linear}
    boundBone {boneName=dyn_coattail_02_r}
    componentAppliedLinearAccClamp=<1000 1000 1000>
    componentLinearAccScale=<0.6 0.6 0.6>
    componentLinearVelScale=<0.2 0.2 0.2>
    physicsBodyDefinitions [{
      collisionType=CustomSphere
      sphereCollisionRadius=5.5
      boundBone {boneName=dyn_coattail_02_r}
      boxExtents=<20 10 20>
      constraintSetup {angularLimitsMax=<5 20 8>, angularLimitsMin=<-5 -20 -8>, angularTarget=<1 0 0>}
      localJointOffset=<15 0 0>
     }]
    relativeSpaceBone {boneName=thigh_r}
    sphericalLimits [
     {limitRadius=7.1, drivingBone {boneName=thigh_twist_01_r}, sphereLocalOffset=<0 0 0>}
     {limitRadius=7.1, drivingBone {boneName=thigh_twist_01_r}, sphereLocalOffset=<-3.5 0 0>}
     {limitRadius=7.1, drivingBone {boneName=thigh_twist_01_r}, sphereLocalOffset=<5 0 0>}
    ]
   }
   Anim Dynamics - Bone: dyn_coattail_02_l(Alpha=1, ExternalForce="0,0,0") {
    angularDampingOverride=1
    angularSpringConstant=800
    bAngularSpring=true
    bGravityOverrideInSimSpace=true
    bOverrideAngularDamping=true
    bUseSphericalLimits=true
    simulationSpace=RootRelative
    alphaBoolBlend {blendOption=Linear}
    boundBone {boneName=dyn_coattail_02_l}
    componentAppliedLinearAccClamp=<1000 1000 1000>
    componentLinearAccScale=<0.6 0.6 0.6>
    componentLinearVelScale=<0.2 0.2 0.2>
    physicsBodyDefinitions [{
      collisionType=CustomSphere
      sphereCollisionRadius=3
      boundBone {boneName=dyn_coattail_02_l}
      boxExtents=<20 10 20>
      constraintSetup {angularLimitsMax=<20 20 20>, angularLimitsMin=<-20 -20 -20>, angularTarget=<1 0 0>}
      localJointOffset=<-15 0 0>
     }]
    planarLimits [{drivingBone {boneName=thigh_l}, planeTransform {location=<0 0 3>, rotation=<rot 0 180 -180>}}]
    relativeSpaceBone {boneName=thigh_r}
    sphericalLimits [
     {limitRadius=7.75, drivingBone {boneName=thigh_twist_01_l}, sphereLocalOffset=<0 0 0>}
     {limitRadius=7.75, drivingBone {boneName=thigh_twist_01_l}, sphereLocalOffset=<-3.5 0 0>}
     {limitRadius=7.75, drivingBone {boneName=thigh_twist_01_l}, sphereLocalOffset=<3.5 0 0>}
    ]
   }
   n1 = "Constraint - thigh_half_l (PP)"!(ComponentPose=Local To Component(), Alpha=1, ConstraintWeights_0=0, ConstraintWeights_1:float=1) -> (Pose) {
    alphaBoolBlend {blendOption=Linear}
    boneToModify {boneName=thigh_half_l}
    constraintSetup [{targetBone {boneName=pelvis}}, {targetBone {boneName=thigh_l}}]
    constraintWeights [1,1]
   }
   n2 = "Constraint - thigh_half_r (PP)"!(ComponentPose=@n1, Alpha=1, ConstraintWeights_0=0, ConstraintWeights_1:float=1) -> (Pose) {
    alphaBoolBlend {blendOption=Linear}
    boneToModify {boneName=thigh_half_r}
    constraintSetup [{targetBone {boneName=pelvis}}, {targetBone {boneName=thigh_r}}]
    constraintWeights [1,1]
   }
   n3 = "Constraint - dyn_coattail_Base_r (RRR)"!(ComponentPose=@n2, Alpha=1, ConstraintWeights_0=0.8, ConstraintWeights_1:float=0.2, ConstraintWeights_2:float=0) -> (Pose) {
    alphaBoolBlend {blendOption=Linear}
    boneToModify {boneName=dyn_coattail_Base_r}
    constraintSetup [
     {transformType=Rotation, targetBone {boneName=thigh_r}}
     {transformType=Rotation, targetBone {boneName=pelvis}}
     {transformType=Rotation, targetBone {boneName=thigh_twist_01_r}}
    ]
    constraintWeights [1,1,1]
   }
   n4 = "Transform (Modify) Bone - Bone: dyn_coattail_01_r"(
    ComponentPose=@n3
    Alpha=Get legSwingDelta_r() -> (legSwingDelta_r:double)
    Translation="-0.806,-4.451,-0.091"
    Rotation="30.073,-26.797,0.429"
    Scale="1,1,1"
   ) {
    rotationMode=BMM_Additive
    rotationSpace=BCS_BoneSpace
    translationMode=BMM_Additive
    translationSpace=BCS_BoneSpace
    alphaBoolBlend {blendOption=Linear}
    alphaScaleBiasClamp {bMapRange=true, inRange {max=0.9}}
    boneToModify {boneName=dyn_coattail_01_r}
    rotation=<rot 30.073 -26.7969 0.429>
    translation=<-0.806 -4.4509 -0.091>
   }
   n5 = Anim Dynamics - Bone: dyn_coattail_02_r(ComponentPose=@n4, Alpha=1, ExternalForce="0,0,0") {
    angularDampingOverride=1
    angularSpringConstant=800
    bAngularSpring=true
    bGravityOverrideInSimSpace=true
    bOverrideAngularDamping=true
    bUseSphericalLimits=true
    simulationSpace=RootRelative
    alphaBoolBlend {blendOption=Linear}
    boundBone {boneName=dyn_coattail_02_r}
    componentAppliedLinearAccClamp=<1000 1000 1000>
    componentLinearAccScale=<0.6 0.6 0.6>
    componentLinearVelScale=<0.2 0.2 0.2>
    physicsBodyDefinitions [{
      collisionType=CustomSphere
      sphereCollisionRadius=4
      boundBone {boneName=dyn_coattail_02_r}
      boxExtents=<20 10 20>
      constraintSetup {angularLimitsMax=<3 20 3>, angularLimitsMin=<-3 -20 -3>, angularTarget=<1 0 0>}
      localJointOffset=<15 0 0>
     }]
    relativeSpaceBone {boneName=thigh_r}
    sphericalLimits [
     {limitRadius=7.75, drivingBone {boneName=thigh_twist_01_r}, sphereLocalOffset=<0 0 0>}
     {limitRadius=7.75, drivingBone {boneName=thigh_twist_01_r}, sphereLocalOffset=<-3.5 0 0>}
     {limitRadius=7.75, drivingBone {boneName=thigh_twist_01_r}, sphereLocalOffset=<5 0 0>}
    ]
   }
   n6 = "Constraint - dyn_coattail_Base_l (RRR)"!(ComponentPose=@n5, Alpha=1, ConstraintWeights_0=0.8, ConstraintWeights_1:float=0.2, ConstraintWeights_2:float=0) -> (Pose) {
    alphaBoolBlend {blendOption=Linear}
    boneToModify {boneName=dyn_coattail_Base_l}
    constraintSetup [
     {transformType=Rotation, targetBone {boneName=thigh_l}}
     {transformType=Rotation, targetBone {boneName=pelvis}}
     {transformType=Rotation, targetBone {boneName=thigh_twist_01_l}}
    ]
    constraintWeights [1,1,1]
   }
   n7 = "Transform (Modify) Bone - Bone: dyn_coattail_01_l"(
    ComponentPose=@n6
    Alpha=Get legSwingDelta_l() -> (legSwingDelta_l:double)
    Translation="2.962,5.671,-2.683"
    Rotation="65.398,28.337,55.613"
    Scale="1,1,1"
   ) {
    rotationMode=BMM_Additive
    rotationSpace=BCS_BoneSpace
    translationMode=BMM_Additive
    translationSpace=BCS_BoneSpace
    alphaBoolBlend {blendOption=Linear}
    alphaScaleBiasClamp {bClampResult=true, clampMin=0.05}
    boneToModify {boneName=dyn_coattail_01_l}
    rotation=<rot 65.398 28.337 55.613>
    translation=<2.962 5.671 -2.683>
   }
   n8 = Anim Dynamics - Bone: dyn_coattail_02_l(ComponentPose=@n7, Alpha=1, ExternalForce="0,0,0") {
    angularDampingOverride=1
    angularSpringConstant=800
    bAngularSpring=true
    bGravityOverrideInSimSpace=true
    bOverrideAngularDamping=true
    bUseSphericalLimits=true
    simulationSpace=RootRelative
    alphaBoolBlend {blendOption=Linear}
    boundBone {boneName=dyn_coattail_02_l}
    componentAppliedLinearAccClamp=<1000 1000 1000>
    componentLinearAccScale=<0.6 0.6 0.6>
    componentLinearVelScale=<0.2 0.2 0.2>
    physicsBodyDefinitions [{
      collisionType=CustomSphere
      sphereCollisionRadius=3
      boundBone {boneName=dyn_coattail_02_l}
      boxExtents=<20 10 20>
      constraintSetup {angularLimitsMax=<20 20 20>, angularLimitsMin=<-20 -20 -20>, angularTarget=<1 0 0>}
      localJointOffset=<-15 0 0>
     }]
    planarLimits [{drivingBone {boneName=thigh_l}, planeTransform {location=<0 0 3>, rotation=<rot 0 -180 -180>}}]
    relativeSpaceBone {boneName=thigh_r}
    sphericalLimits [
     {limitRadius=7.75, drivingBone {boneName=thigh_twist_01_l}, sphereLocalOffset=<0 0 0>}
     {limitRadius=7.75, drivingBone {boneName=thigh_twist_01_l}, sphereLocalOffset=<-3.5 0 0>}
     {limitRadius=7.75, drivingBone {boneName=thigh_twist_01_l}, sphereLocalOffset=<3.5 0 0>}
    ]
   }
   n9 = Get jumping_ext_force() -> (jumping_ext_force:Vector)
   n10 = Anim Dynamics - Bone: dyn_bag_01(ComponentPose=@n8, Alpha=1, ExternalForce=@n9) {
    angularDampingOverride=1
    angularSpringConstant=250
    bGravityOverrideInSimSpace=true
    bOverrideAngularDamping=true
    simulationSpace=RootRelative
    alphaBoolBlend {blendOption=Linear}
    boundBone {boneName=dyn_bag_01}
    componentAppliedLinearAccClamp=<1000 1000 1000>
    componentLinearAccScale=<0.6 0.6 0.6>
    componentLinearVelScale=<0.4 0.4 0.4>
    physicsBodyDefinitions [{
      sphereCollisionRadius=0
      boundBone {boneName=dyn_bag_01}
      boxExtents=<8 8 16>
      constraintSetup {angularLimitsMax=<10 30 2>, angularLimitsMin=<-10 -30 -45>, angularTarget=<1 0 0>}
      localJointOffset=<6 0 0>
     }]
   }
   n11 = Anim Dynamics - Bone: dyn_bag_02(ComponentPose=@n10, Alpha=1, ExternalForce=@n9) {
    angularDampingOverride=1
    angularSpringConstant=250
    bAngularSpring=true
    bGravityOverrideInSimSpace=true
    bOverrideAngularDamping=true
    simulationSpace=RootRelative
    alphaBoolBlend {blendOption=Linear}
    boundBone {boneName=dyn_bag_02}
    componentAppliedLinearAccClamp=<1000 1000 1000>
    componentLinearAccScale=<0.6 0.6 0.6>
    componentLinearVelScale=<0.4 0.4 0.4>
    physicsBodyDefinitions [{
      sphereCollisionRadius=0
      boundBone {boneName=dyn_bag_02}
      boxExtents=<12 20 12>
      constraintSetup {angularLimitsMax=<10 30 2>, angularLimitsMin=<-10 -30 -45>, angularTarget=<1 0 0>}
      localJointOffset=<6 0 0>
     }]
    relativeSpaceBone {boneName=dyn_bag_01}
   }
   n12 = RigidBody(ComponentPose=@n11) {
    bClampLinearTranslationLimitToRefPose=true
    bOverrideWorldGravity=true
    overridePhysicsAsset=&/M_MED_MaskedWarrior_RPhysics
    simulationSpace=BaseBoneSpace
    alphaBoolBlend {blendOption=Linear}
    baseBoneRef {boneName=thigh_half_r}
    componentAppliedLinearAccClamp=<1500 1500 1500>
    componentLinearAccScale=<0.65 0.65 0.65>
    componentLinearVelScale=<0.4 0.4 0.4>
    overrideWorldGravity {z=-980}
   }
   n13 = RigidBody(ComponentPose=@n12) {
    bClampLinearTranslationLimitToRefPose=true
    bOverrideWorldGravity=true
    overridePhysicsAsset=&/M_MED_MaskedWarrior_LPhysics
    simulationSpace=BaseBoneSpace
    alphaBoolBlend {blendOption=Linear}
    baseBoneRef {boneName=thigh_half_l}
    componentAppliedLinearAccClamp=<1500 1500 1500>
    componentLinearAccScale=<0.65 0.65 0.65>
    componentLinearVelScale=<0.4 0.4 0.4>
    overrideWorldGravity {z=-980}
   }
   n14 = RigidBody(ComponentPose=@n13) {
    bClampLinearTranslationLimitToRefPose=true
    bOverrideWorldGravity=true
    overridePhysicsAsset=&/M_MED_MaskedWarrior_CPhysics
    simulationSpace=BaseBoneSpace
    alphaBoolBlend {blendOption=Linear}
    baseBoneRef {boneName=pelvis}
    componentAppliedLinearAccClamp=<1500 1500 1500>
    componentLinearAccScale=<0.65 0.65 0.65>
    componentLinearVelScale=<0.4 0.4 0.4>
    overrideWorldGravity {z=-980}
   }
   Component To Local(ComponentPose=@n14)
   n15 = RigidBody() {
    bClampLinearTranslationLimitToRefPose=true
    bOverrideWorldGravity=true
    overridePhysicsAsset=&M_MED_MaskedWarrior_LPhysics
    simulationSpace=BaseBoneSpace
    alphaBoolBlend {blendOption=Linear}
    baseBoneRef {boneName=thigh_half_l}
    componentLinearAccScale=<0.5 0.5 0.5>
    componentLinearVelScale=<3 0.3 0.3>
    overrideWorldGravity {z=-980}
   }
   Component To Local(ComponentPose=@n15)
   n16 = "Transform (Modify) Bone - Bone: dyn_coattail_01_r"(Alpha=1, Translation="-0.806,-4.451,-0.091", Rotation="30.073,-26.797,0.429", Scale="1,1,1") {
    rotationMode=BMM_Additive
    rotationSpace=BCS_BoneSpace
    translationMode=BMM_Additive
    translationSpace=BCS_BoneSpace
    alphaBoolBlend {blendOption=Linear}
    alphaScaleBiasClamp {bMapRange=true, inRange {max=0.9}}
    boneToModify {boneName=dyn_coattail_01_r}
    rotation=<rot 30.073 -26.7969 0.429>
    translation=<-0.806 -4.4509 -0.091>
   }
   "Transform (Modify) Bone - Bone: dyn_coattail_01_l"(ComponentPose=@n16, Alpha=1, Translation="2.962,5.671,-2.683", Rotation="65.398,28.337,55.613", Scale="1,1,1") {
    rotationMode=BMM_Additive
    rotationSpace=BCS_BoneSpace
    translationMode=BMM_Additive
    translationSpace=BCS_BoneSpace
    alphaBoolBlend {blendOption=Linear}
    alphaScaleBiasClamp {bClampResult=true, clampMin=0.05}
    boneToModify {boneName=dyn_coattail_01_l}
    rotation=<rot 65.398 28.337 55.613>
    translation=<2.962 5.671 -2.683>
   }
   n17 = RigidBody() {
    bClampLinearTranslationLimitToRefPose=true
    bOverrideWorldGravity=true
    overridePhysicsAsset=&M_MED_MaskedWarrior_RPhysics
    simulationSpace=BaseBoneSpace
    alphaBoolBlend {blendOption=Linear}
    baseBoneRef {boneName=thigh_half_r}
    componentAppliedLinearAccClamp=<1500 1500 1500>
    componentLinearAccScale=<1 1 1>
    componentLinearVelScale=<0.35 0.35 0.35>
    overrideWorldGravity {z=-1500}
   }
   RigidBody(ComponentPose=@n17) {
    bClampLinearTranslationLimitToRefPose=true
    bOverrideWorldGravity=true
    overridePhysicsAsset=&M_MED_MaskedWarrior_LPhysics
    simulationSpace=BaseBoneSpace
    alphaBoolBlend {blendOption=Linear}
    baseBoneRef {boneName=thigh_half_l}
    componentAppliedLinearAccClamp=<1500 1500 1500>
    componentLinearAccScale=<1 1 1>
    componentLinearVelScale=<0.4 0.4 0.4>
    overrideWorldGravity {z=-1500}
   }
   n18 = "Transform (Modify) Bone - Bone: dyn_coattail_01_r"(
    ComponentPose=Local To Component(LocalPose=Use cached pose 'preJump'())
    Alpha=Get legSwingDelta_r()
    Translation="-0.397,-6.411,-2.693"
    Rotation="-26.731,-8.945,-17.730"
    Scale="1,1,1"
   ) {
    rotationMode=BMM_Additive
    rotationSpace=BCS_BoneSpace
    translationMode=BMM_Additive
    translationSpace=BCS_BoneSpace
    alphaBoolBlend {blendOption=Linear}
    alphaScaleBiasClamp {bMapRange=true, inRange {max=0.9}}
    boneToModify {boneName=dyn_coattail_01_r}
    rotation=<rot -26.731 -8.9451 -17.73>
    translation=<-0.397 -6.411 -2.693>
   }
   n19 = "Transform (Modify) Bone - Bone: dyn_coattail_01_l"(ComponentPose=@n18, Alpha=Get legSwingDelta_l(), Translation="-0.000,4.280,0.000", Rotation="-17.734,-13.015,4.029", Scale="1,1,1") {
    rotationMode=BMM_Additive
    rotationSpace=BCS_BoneSpace
    translationMode=BMM_Additive
    translationSpace=BCS_BoneSpace
    alphaBoolBlend {blendOption=Linear}
    alphaScaleBiasClamp {bClampResult=true, clampMin=0.05}
    boneToModify {boneName=dyn_coattail_01_l}
    rotation=<rot -17.734 -13.0151 4.029>
    translation {y=4.28}
   }
   n20 = Component To Local(ComponentPose=@n19)
   n21 = "Constraint - dyn_coattail_Base_r (RRR)"(
    ComponentPose=Local To Component(LocalPose=Use cached pose 'preJump'())
    Alpha=1
    ConstraintWeights_0=0.5
    ConstraintWeights_1=0.5
    ConstraintWeights_2=0
   ) {
    alphaBoolBlend {blendOption=Linear}
    boneToModify {boneName=dyn_coattail_Base_r}
    constraintSetup [
     {transformType=Rotation, targetBone {boneName=thigh_r}}
     {transformType=Rotation, targetBone {boneName=pelvis}}
     {transformType=Rotation, targetBone {boneName=thigh_twist_01_r}}
    ]
    constraintWeights [1,1,1]
   }
   n22 = "Transform (Modify) Bone - Bone: dyn_coattail_01_r"(ComponentPose=@n21, Alpha=Get legSwingDelta_r(), Translation="-0.258,-7.220,-6.698", Rotation="-31.697,-12.468,-25.014", Scale="1,1,1") {
    rotationMode=BMM_Additive
    rotationSpace=BCS_BoneSpace
    translationMode=BMM_Additive
    translationSpace=BCS_BoneSpace
    alphaBoolBlend {blendOption=Linear}
    alphaScaleBiasClamp {bMapRange=true, inRange {max=0.85, min=0.1}}
    boneToModify {boneName=dyn_coattail_01_r}
    rotation=<rot -31.697 -12.4681 -25.0139>
    translation=<-0.258 -7.22 -6.698>
   }
   n23 = "Transform (Modify) Bone - Bone: dyn_coattail_01_l"(ComponentPose=@n22, Alpha=Get legSwingDelta_l(), Translation="-3.368,8.501,4.672", Rotation="-9.394,53.706,-72.762", Scale="1,1,1") {
    rotationMode=BMM_Additive
    rotationSpace=BCS_BoneSpace
    translationMode=BMM_Additive
    translationSpace=BCS_BoneSpace
    alphaBoolBlend {blendOption=Linear}
    alphaScaleBiasClamp {bClampResult=true}
    boneToModify {boneName=dyn_coattail_01_l}
    rotation=<rot -9.394 53.706 -72.7619>
    translation=<-3.368 8.501 4.672>
   }
   n24 = Component To Local(ComponentPose=@n23)
   n25 = Get airLeg() -> (airLeg:bool)
   n26 = RigidBody(ComponentPose=Local To Component(LocalPose=Blend Poses by bool(BlendPose_0=@n20, BlendPose_1=@n24, BlendTime_0=0.2, BlendTime_1=0.2, bActiveValue=@n25) {blendType=Linear})) {
    bClampLinearTranslationLimitToRefPose=true
    bOverrideWorldGravity=true
    overridePhysicsAsset=&M_MED_MaskedWarrior_RPhysics
    simulationSpace=BaseBoneSpace
    alphaBoolBlend {blendOption=Linear}
    baseBoneRef {boneName=thigh_half_r}
    componentAppliedLinearAccClamp=<1500 1500 1500>
    componentLinearAccScale=<1 1 1>
    componentLinearVelScale=<0.35 0.35 0.35>
    overrideWorldGravity {z=-1500}
   }
   n27 = RigidBody(ComponentPose=@n26) {
    bClampLinearTranslationLimitToRefPose=true
    bOverrideWorldGravity=true
    overridePhysicsAsset=&M_MED_MaskedWarrior_LPhysics
    simulationSpace=BaseBoneSpace
    alphaBoolBlend {blendOption=Linear}
    baseBoneRef {boneName=thigh_half_l}
    componentAppliedLinearAccClamp=<1500 1500 1500>
    componentLinearAccScale=<1 1 1>
    componentLinearVelScale=<0.4 0.4 0.4>
    overrideWorldGravity {z=-1500}
   }
   n28 = Get headSwingDelta() -> (headSwingDelta:double)
   n29 = "Transform (Modify) Bone - Bone: dyn_collarBack_Base"(ComponentPose=@n27, Alpha=@n28, Translation="0,0,0", Rotation="0.000,-25,0.000", Scale="1,1,1") {
    rotationMode=BMM_Additive
    rotationSpace=BCS_BoneSpace
    translationMode=BMM_Additive
    translationSpace=BCS_BoneSpace
    alphaBoolBlend {blendOption=Linear}
    alphaScaleBiasClamp {bMapRange=true, inRange {max=0.85, min=0.35}}
    boneToModify {boneName=dyn_collarBack_Base}
    rotation {yaw=-25}
   }
   n30 = "Transform (Modify) Bone - Bone: dyn_collarBack_Base_r"(ComponentPose=@n29, Alpha=@n28, Translation="0,-3,1", Rotation="20,-75,25", Scale="1,1,1") {
    rotationMode=BMM_Additive
    rotationSpace=BCS_BoneSpace
    translationMode=BMM_Additive
    translationSpace=BCS_BoneSpace
    alphaBoolBlend {blendOption=Linear}
    alphaScaleBiasClamp {bMapRange=true, inRange {max=0.85, min=0.35}}
    boneToModify {boneName=dyn_collarBack_Base_r}
    rotation=<rot 20 -75 25>
    translation {y=-3, z=1}
   }
   n31 = "Transform (Modify) Bone - Bone: dyn_collarBack_Base_l"(ComponentPose=@n30, Alpha=@n28, Translation="-2,3,0.000", Rotation="0.000,-60,20.405", Scale="1,1,1") {
    rotationMode=BMM_Additive
    rotationSpace=BCS_BoneSpace
    translationMode=BMM_Additive
    translationSpace=BCS_BoneSpace
    alphaBoolBlend {blendOption=Linear}
    alphaScaleBiasClamp {bMapRange=true, inRange {max=0.85, min=0.35}}
    boneToModify {boneName=dyn_collarBack_Base_l}
    rotation {roll=20.405, yaw=-60}
    translation=<-2 3>
   }
   n32 = Get armSwingDelta_r() -> (armSwingDelta_r:double)
   n33 = "Transform (Modify) Bone - Bone: dyn_collarFront_Base_r"(ComponentPose=@n31, Alpha=@n32, Translation="1,3,0.000", Rotation="0.000,70,5", Scale="1,1,1") {
    rotationMode=BMM_Additive
    rotationSpace=BCS_BoneSpace
    translationMode=BMM_Additive
    translationSpace=BCS_BoneSpace
    alphaBoolBlend {blendOption=Linear}
    alphaScaleBiasClamp {bMapRange=true, inRange {max=0.65, min=0.25}}
    boneToModify {boneName=dyn_collarFront_Base_r}
    rotation {roll=5, yaw=70}
    translation=<1 3>
   }
   n34 = "Transform (Modify) Bone - Bone: dyn_collarFront_Base_r"(ComponentPose=@n33, Alpha=@n32, Translation="0,0,0", Rotation="0,0,0", Scale="1,0.2,1") {
    rotationSpace=BCS_BoneSpace
    scaleMode=BMM_Replace
    scaleSpace=BCS_BoneSpace
    translationSpace=BCS_BoneSpace
    alphaBoolBlend {blendOption=Linear}
    alphaScaleBiasClamp {bMapRange=true, inRange {max=0.65, min=0.25}}
    boneToModify {boneName=dyn_collarFront_Base_r}
    scale {y=0.2}
   }
   n35 = "Transform (Modify) Bone - Bone: dyn_collarSide_Base_l"(ComponentPose=@n34, Alpha=Get armSwingDelta_l() -> (armSwingDelta_l:double), Translation="0,0,0", Rotation="0.000,45,0.000", Scale="1,1,1") {
    rotationMode=BMM_Additive
    rotationSpace=BCS_BoneSpace
    translationMode=BMM_Additive
    translationSpace=BCS_BoneSpace
    alphaBoolBlend {blendOption=Linear}
    alphaScaleBiasClamp {bMapRange=true, inRange {max=0.85, min=0.5}}
    boneToModify {boneName=dyn_collarSide_Base_l}
    rotation {yaw=45}
   }
   n36 = "Transform (Modify) Bone - Bone: dyn_collarBack_Base"(ComponentPose=@n35, Alpha=Get headTiltXTREME() -> (headTiltXTREME:double), Translation="0,-5,0", Rotation="0.000,-35,0.000", Scale="1,1,1") {
    rotationMode=BMM_Additive
    rotationSpace=BCS_BoneSpace
    translationMode=BMM_Additive
    translationSpace=BCS_BoneSpace
    alphaBoolBlend {blendOption=Linear}
    alphaScaleBiasClamp {bMapRange=true}
    boneToModify {boneName=dyn_collarBack_Base}
    rotation {yaw=-35}
    translation {y=-5}
   }
   n37 = RigidBody(ComponentPose=@n36) {
    bClampLinearTranslationLimitToRefPose=true
    bOverrideWorldGravity=true
    overridePhysicsAsset=&M_MED_MaskedWarrior_CPhysics
    simulationSpace=BaseBoneSpace
    alphaBoolBlend {blendOption=Linear}
    baseBoneRef {boneName=pelvis}
    componentAppliedLinearAccClamp=<1500 1500 1500>
    componentLinearAccScale=<0.65 0.65 0.65>
    componentLinearVelScale=<0.4 0.4 0.4>
    overrideWorldGravity {z=-980}
   }
   n38 = Look At - Bone: dyn_shoeTongue_Base_r(ComponentPose=@n37, Alpha=1) {
    alphaBoolBlend {blendOption=Linear}
    boneToModify {boneName=dyn_shoeTongue_Base_r}
    lookAtLocation=<0 -3 1.5>
    lookAtTarget {boneReference {boneName=calf_twist_02_r}}
    lookAt_Axis {axis=<1 0>}
    lookUp_Axis {axis=<0 1>}
   }
   n39 = Look At - Bone: dyn_shoeTongue_Base_l(ComponentPose=@n38, Alpha=1) {
    alphaBoolBlend {blendOption=Linear}
    boneToModify {boneName=dyn_shoeTongue_Base_l}
    lookAtLocation=<0 3 -1.5>
    lookAtTarget {boneReference {boneName=calf_twist_02_l}}
    lookAt_Axis {axis=<-1 0>}
    lookUp_Axis {axis=<0 -1>}
   }
   Output Pose(Result=Component To Local(ComponentPose=@n39)) {layerGroup=None}
   n40 = "Constraint - thigh_half_l (PP)"(
    ComponentPose=Local To Component(LocalPose=Copy Pose From Mesh() {bUseAttachedParent=true})
    Alpha=0.65
    ConstraintWeights_0=0
    ConstraintWeights_1=1
   ) {
    alphaBoolBlend {blendOption=Linear}
    boneToModify {boneName=thigh_half_l}
    constraintSetup [{targetBone {boneName=pelvis}}, {targetBone {boneName=thigh_l}}]
    constraintWeights [1,1]
   }
   n41 = "Constraint - thigh_half_r (PP)"(ComponentPose=@n40, Alpha=0.65, ConstraintWeights_0=0, ConstraintWeights_1=1) {
    alphaBoolBlend {blendOption=Linear}
    boneToModify {boneName=thigh_half_r}
    constraintSetup [{targetBone {boneName=pelvis}}, {targetBone {boneName=thigh_r}}]
    constraintWeights [1,1]
   }
   n42 = "Constraint - dyn_coattail_Base_r (RRR)"(ComponentPose=@n41, Alpha=Get legSwingDelta_r(), ConstraintWeights_0=0.8, ConstraintWeights_1=0.2, ConstraintWeights_2=0) {
    alphaBoolBlend {blendOption=Linear}
    alphaScaleBiasClamp {bClampResult=true, clampMax=0.8, clampMin=0.2}
    boneToModify {boneName=dyn_coattail_Base_r}
    constraintSetup [
     {transformType=Rotation, targetBone {boneName=thigh_r}}
     {transformType=Rotation, targetBone {boneName=pelvis}}
     {transformType=Rotation, targetBone {boneName=thigh_twist_01_r}}
    ]
    constraintWeights [1,1,1]
   }
   n43 = "Constraint - dyn_coattail_Base_l (RRR)"(ComponentPose=@n42, Alpha=Get legSwingDelta_l(), ConstraintWeights_0=0.8, ConstraintWeights_1=0.2, ConstraintWeights_2=0) {
    alphaBoolBlend {blendOption=Linear}
    alphaScaleBiasClamp {bClampResult=true, clampMin=0.5}
    boneToModify {boneName=dyn_coattail_Base_l}
    constraintSetup [
     {transformType=Rotation, targetBone {boneName=thigh_l}}
     {transformType=Rotation, targetBone {boneName=pelvis}}
     {transformType=Rotation, targetBone {boneName=thigh_twist_01_l}}
    ]
    constraintWeights [1,1,1]
   }
   n44 = Get jumping_ext_force()
   n45 = Anim Dynamics - Bone: dyn_bag_01(ComponentPose=@n43, Alpha=1, ExternalForce=@n44) {
    angularDampingOverride=1
    angularSpringConstant=250
    bGravityOverrideInSimSpace=true
    bOverrideAngularDamping=true
    simulationSpace=RootRelative
    alphaBoolBlend {blendOption=Linear}
    boundBone {boneName=dyn_bag_01}
    componentAppliedLinearAccClamp=<1000 1000 1000>
    componentLinearAccScale=<0.6 0.6 0.6>
    componentLinearVelScale=<0.4 0.4 0.4>
    physicsBodyDefinitions [{
      sphereCollisionRadius=0
      boundBone {boneName=dyn_bag_01}
      boxExtents=<10 10 18>
      constraintSetup {angularLimitsMax=<10 30 2>, angularLimitsMin=<-10 -30 -45>, angularTarget=<1 0 0>}
      localJointOffset=<7 0 0>
     }]
   }
   n46 = Anim Dynamics - Bone: dyn_bag_02(ComponentPose=@n45, Alpha=1, ExternalForce=@n44) {
    angularDampingOverride=1
    angularSpringConstant=250
    bAngularSpring=true
    bGravityOverrideInSimSpace=true
    bOverrideAngularDamping=true
    simulationSpace=RootRelative
    alphaBoolBlend {blendOption=Linear}
    boundBone {boneName=dyn_bag_02}
    componentAppliedLinearAccClamp=<1000 1000 1000>
    componentLinearAccScale=<0.6 0.6 0.6>
    componentLinearVelScale=<0.4 0.4 0.4>
    physicsBodyDefinitions [{
      sphereCollisionRadius=0
      boundBone {boneName=dyn_bag_02}
      boxExtents=<14 22 14>
      constraintSetup {angularLimitsMax=<10 30 2>, angularLimitsMin=<-10 -30 -45>, angularTarget=<1 0 0>}
      localJointOffset=<8 0 0>
     }]
    relativeSpaceBone {boneName=dyn_bag_01}
   }
   Save cached pose 'preJump'(Pose=Component To Local(ComponentPose=@n46)) {cacheName=preJump}
  ]}
 EventGraph {nodes [
   Event BlueprintUpdateAnimation()
   > n1 = Sequence!(execute) -> (then_0, then_1, then_2:Exec, then_3:Exec)
   > n2 = Branch(Condition=Get bIsJumping())
   > Delay(Duration=0.25)
   > Set jumping_force_vector(jumping_force_vector:Vector="0.25,2.5,0.1") -> (Output_Get:Vector)
   > Set jumping_ext_force(jumping_ext_force:Vector="0,0,2000") -> (Output_Get:Vector)
   > n3 = Delay(Duration=0.875)
   Do Once(execute=@n2.else, Reset=@n3.then)
   > Set jumping_ext_force(jumping_ext_force="0, 0, 0")
   Set airLeg(execute=@n1.then_3, airLeg:bool=OR Boolean(A=Get bIsFalling(), B=Get bIsJumping())) -> (Output_Get:bool)
   n4 = GetOwnerSkeletalMeshComponent()
   Set legSwingDelta_r(
    execute=@n1.then_1
    legSwingDelta_r:double=MapRangeClamped(Value=GetTwistAndSwingAngleOfDeltaRotationFromRefPose(self=@n4, BoneName=thigh_r).OutSwingAngle, InRangeA=173, InRangeB=71)
   ) -> (Output_Get:double)
   > Set legSwingDelta_l(legSwingDelta_l:double=MapRangeClamped(Value=GetTwistAndSwingAngleOfDeltaRotationFromRefPose(self=@n4, BoneName=thigh_l).OutSwingAngle, InRangeA=24, InRangeB=101)) -> (Output_Get:double)
   n5 = GetOwnerSkeletalMeshComponent()
   n6 = BreakTransform(InTransform=GetDeltaTransformFromRefPose(self=@n5, BoneName=head, BaseName=spine_05))
   Set headSwingDelta(execute=@n1.then_2, headSwingDelta:double=MapRangeClamped(Value=BreakRotator(InRot=@n6.Rotation).Yaw, InRangeA=17, InRangeB=-52)) -> (Output_Get:double)
   > Set headTiltXTREME(headTiltXTREME:double=MapRangeClamped(Value=BreakRotator(InRot=@n6.Rotation).Yaw, InRangeA=-52, InRangeB=-72)) -> (Output_Get:double)
   n7 = BreakRotator(InRot=@n6.Rotation)
   > Set headTwist_r(headTwist_r:double=MapRangeClamped(Value=@n7, InRangeA=14, InRangeB=25)) -> (Output_Get:double)
   > Set headTwist_l(headTwist_l:double=MapRangeClamped(Value=@n7, InRangeA=-9, InRangeB=-20)) -> (Output_Get:double)
   n8 = BreakRotator(InRot=BreakTransform(InTransform=GetDeltaTransformFromRefPose(self=@n5, BoneName=upperarm_r, BaseName=spine_05)).Rotation)
   > Set armSwingDelta_r(armSwingDelta_r:double=MapRangeClamped(Value="Absolute (Float)"(A=@n8.Yaw), InRangeA=45, InRangeB=135)) -> (Output_Get:double)
   n9 = BreakRotator(InRot=BreakTransform(InTransform=GetDeltaTransformFromRefPose(self=@n5, BoneName=upperarm_l, BaseName=spine_05)).Rotation)
   > Set armSwingDelta_l(armSwingDelta_l:double=MapRangeClamped(Value="Absolute (Float)"(A=@n9.Yaw), InRangeA=16, InRangeB=105)) -> (Output_Get:double)
   TryGetPawnOwner()
   Reroute Node(InputPin:Wildcard) -> (OutputPin:Wildcard)
  ]}
]
variables [
 airLeg {type=boolean}
 armSwingDelta_l {type=number}
 armSwingDelta_r {type=number}
 headSwingDelta {type=number}
 headTiltXTREME {type=number}
 headTwist_l {type=number}
 headTwist_r {type=number}
 jumping_ext_force {type=Vector}
 jumping_force_vector {type=Vector}
 legSwingDelta_l {type=number}
 legSwingDelta_r {type=number}
]
